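# MI_Backpack_PixieParts_Hide is a base Material, despite its MI_ prefix, for the PixieParts Lava backpack cosmetic, named for hiding parts of it.
# It uses Masked blend and Unlit shading, is enabled for skeletal meshes, and turns off automatic usage flagging.
# The graph has no expressions, so nothing drives emissive color or opacity mask and both stay at engine defaults.
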Material /BRCosmetics/Accessories/FORT_Backpacks/Backpack_M_MED_PixieParts_Lava/Materials/MI_Backpack_PixieParts_Hide
graph {
 expressions {}
 properties {bAutomaticallySetUsageInEditor=false, bUsedWithSkeletalMesh=true, blendMode=BLEND_Masked, shadingModel=MSM_Unlit, usedShadingModels=Unlit}
}
