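# Material function that turns an incoming material into refractive, volumetric glass for the PixieParts Lava cosmetic character body.
# It adjusts surface normal, roughness and specular, then estimates view-dependent thickness and samples the scene through screen-space refraction or approximates it for alpha blending.
# A dielectric interior volume function combines these with absorption, scattering, tinted-glass masking and sun/sky lighting.
# A shading-path switch controls alpha blending, with separate mobile tuning parameters.
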MaterialFunction /BRCosmetics/Characters/Player/Male/Medium/Bodies/M_MED_PixieParts_Lava/Materials/MaterialFunctions/MF_PixieParts_Glass
graph {
 expressions {
  MaterialExpressionAdd_0 {
   class=MaterialExpressionAdd
   inputs [
    A {from=MaterialExpressionStaticSwitch_3}
    B {from=MaterialExpressionNamedRerouteUsage_4}
   ]
  }
  MaterialExpressionAppendVector_3 {
   class=MaterialExpressionAppendVector
   inputs [
    A {from=MaterialExpressionNamedRerouteUsage_73, output=""}
    B {from=MaterialExpressionConstant_8, output=""}
   ]
  }
  MaterialExpressionAppendVector_4 {
   class=MaterialExpressionAppendVector
   inputs [
    A {from=MaterialExpressionNamedRerouteUsage_159, output=""}
    B {from=MaterialExpressionConstant_9, output=""}
   ]
  }
  MaterialExpressionAppendVector_7 {
   class=MaterialExpressionAppendVector
   inputs [
    A {from=MaterialExpressionMultiply_19, output=""}
    B {from=MaterialExpressionMaterialFunctionCall_13, output="Distance (S)"}
   ]
   properties {desc=SceneColor}
  }
  MaterialExpressionAppendVector_8 {
   class=MaterialExpressionAppendVector
   inputs [
    A {from=MaterialExpressionMaterialFunctionCall_20, output="Scene Luminance (V3)"}
    B {from=MaterialExpressionMaterialFunctionCall_14, output="Distance (S)"}
   ]
   properties {desc=AlphaBlending}
  }
  MaterialExpressionCameraVectorWS_5 {class=MaterialExpressionCameraVectorWS, inputs [], properties {}}
  MaterialExpressionChannelMaskParameter_0 {
   class=MaterialExpressionChannelMaskParameter
   inputs [None {from=MaterialExpressionFunctionInput_2, output=""}]
   properties {group=Glass | 03_Volume, parameterName=Volume - Tinted Glass Mask Channel, sortPriority=100}
  }
  MaterialExpressionConstant_0 {class=MaterialExpressionConstant, properties {r=1}}
  MaterialExpressionConstant_10 {class=MaterialExpressionConstant, properties {r=1}}
  MaterialExpressionConstant_14 {class=MaterialExpressionConstant}
  MaterialExpressionConstant_15 {class=MaterialExpressionConstant, properties {r=1}}
  MaterialExpressionConstant_17 {class=MaterialExpressionConstant, properties {desc=Metallic}}
  MaterialExpressionConstant_8 {class=MaterialExpressionConstant}
  MaterialExpressionConstant_9 {class=MaterialExpressionConstant, properties {r=1}}
  MaterialExpressionConvert_11 {
   class=MaterialExpressionConvert
   inputs [None {from=MaterialExpressionStaticSwitch_1}]
   properties {
    nodeName=Break Float3
    convertInputs [{
      type=Vector3
      expressionInput {
       expression=&self:MaterialExpressionStaticSwitch_1
       inputName=None
       mask=0
       maskA=0
       maskB=0
       maskG=0
       maskR=0
       outputIndex=0
      }
     }]
    convertOutputs [{type=Scalar}, {type=Scalar}, {type=Scalar}]
   }
  }
  MaterialExpressionConvert_4 {
   class=MaterialExpressionConvert
   inputs [None {from=MaterialExpressionStaticSwitch_5}]
   properties {
    convertInputs [{
      type=Vector4
      defaultValue=<rgba 0 0 0 1>
      expressionInput {
       expression=&self:MaterialExpressionStaticSwitch_5
       inputName=None
       mask=0
       maskA=0
       maskB=0
       maskG=0
       maskR=0
       outputIndex=0
      }
     }]
    convertOutputs [{type=Vector3}, {type=Scalar}]
   }
  }
  MaterialExpressionConvert_5 {class=MaterialExpressionConvert, inputs [], properties {convertOutputs [{type=Vector3, defaultValue=<rgba 0 0 1 1>}]}}
  MaterialExpressionConvert_7 {
   class=MaterialExpressionConvert
   inputs [None {from=MaterialExpressionIf_0}]
   properties {
    nodeName=Break Float2
    convertInputs [{
      type=Vector2
      expressionInput {
       expression=&self:MaterialExpressionIf_0
       inputName=None
       mask=0
       maskA=0
       maskB=0
       maskG=0
       maskR=0
       outputIndex=0
      }
     }]
    convertOutputs [{type=Scalar}, {type=Scalar}]
   }
  }
  MaterialExpressionConvert_8 {
   class=MaterialExpressionConvert
   inputs [
    None {from=MaterialExpressionMaterialFunctionCall_0, output="Metallic (S)"}
    None {from=MaterialExpressionMaterialFunctionCall_0, output="Metallic (S)"}
    None {from=MaterialExpressionMaterialFunctionCall_0, output="Metallic (S)"}
   ]
   properties {
    nodeName=Make Vector3
    convertInputs [
     {
      type=Scalar
      expressionInput {
       expression=&self:MaterialExpressionMaterialFunctionCall_0
       inputName=None
       mask=0
       maskA=0
       maskB=0
       maskG=0
       maskR=0
       outputIndex=1
      }
     }
     {
      type=Scalar
      expressionInput {
       expression=&self:MaterialExpressionMaterialFunctionCall_0
       inputName=None
       mask=0
       maskA=0
       maskB=0
       maskG=0
       maskR=0
       outputIndex=2
      }
     }
     {
      type=Scalar
      expressionInput {
       expression=&self:MaterialExpressionMaterialFunctionCall_0
       inputName=None
       mask=0
       maskA=0
       maskB=0
       maskG=0
       maskR=0
       outputIndex=3
      }
     }
    ]
    convertOutputs [{type=Vector3}]
   }
  }
  MaterialExpressionConvert_9 {
   class=MaterialExpressionConvert
   inputs [
    None {from=MaterialExpressionConstant_17}
    None {from=MaterialExpressionNamedRerouteUsage_59}
    None {from=MaterialExpressionNamedRerouteUsage_63}
   ]
   properties {
    nodeName=Make Vector3
    convertInputs [
     {
      type=Scalar
      expressionInput {
       expression=&self:MaterialExpressionConstant_17
       inputName=None
       mask=0
       maskA=0
       maskB=0
       maskG=0
       maskR=0
       outputIndex=0
      }
     }
     {
      type=Scalar
      expressionInput {
       expression=&self:MaterialExpressionNamedRerouteUsage_59
       inputName=None
       mask=0
       maskA=0
       maskB=0
       maskG=0
       maskR=0
       outputIndex=0
      }
     }
     {
      type=Scalar
      expressionInput {
       expression=&self:MaterialExpressionNamedRerouteUsage_63
       inputName=None
       mask=0
       maskA=0
       maskB=0
       maskG=0
       maskR=0
       outputIndex=0
      }
     }
    ]
    convertOutputs [{type=Vector3}]
   }
  }
  MaterialExpressionEyeAdaptation_0 {class=MaterialExpressionEyeAdaptation, inputs [], properties {}}
  MaterialExpressionFunctionInput_0 {
   class=MaterialExpressionFunctionInput
   inputs [Preview {from="", output=""}]
   properties {inputName=Material, inputType=FunctionInput_MaterialAttributes}
  }
  MaterialExpressionFunctionInput_1 {
   class=MaterialExpressionFunctionInput
   inputs [Preview {from="", output=""}]
   properties {bUsePreviewValueAsDefault=true, inputName=Surface Opacity, inputType=FunctionInput_Scalar, sortPriority=20}
  }
  MaterialExpressionFunctionInput_2 {
   class=MaterialExpressionFunctionInput
   inputs [Preview {from="", output=""}]
   properties {bUsePreviewValueAsDefault=true, inputName=TechArt Mask, sortPriority=10}
  }
  MaterialExpressionFunctionOutput_0 {
   class=MaterialExpressionFunctionOutput
   inputs [None {from=MaterialExpressionSetMaterialAttributes_2, output=""}]
   properties {outputName="Result (MA)"}
  }
  MaterialExpressionGetMaterialAttributes_0 {
   class=MaterialExpressionGetMaterialAttributes
   inputs [None {from=MaterialExpressionNamedRerouteDeclaration_1, output=""}]
   properties {attributeGetTypes [
     #1
     #2
     #3
     #4
     #5
    ]}
  }
  MaterialExpressionGetMaterialAttributes_1 {
   class=MaterialExpressionGetMaterialAttributes
   inputs [None {from=MaterialExpressionSetMaterialAttributes_4, output=""}]
   properties {attributeGetTypes [#6]}
  }
  MaterialExpressionIf_0 {
   class=MaterialExpressionIf
   inputs [
    A {from=MaterialExpressionNamedRerouteUsage_158, output=""}
    B {from=MaterialExpressionNamedRerouteUsage_64, output=""}
    "A > B" {from=MaterialExpressionAppendVector_3, output=""}
    "A == B" {from="", output=""}
    "A < B" {from=MaterialExpressionAppendVector_4, output=""}
   ]
   properties {}
  }
  MaterialExpressionLinearInterpolate_1 {
   class=MaterialExpressionLinearInterpolate
   inputs [
    A {from=MaterialExpressionNamedRerouteUsage_41}
    B {from=MaterialExpressionConvert_5}
    Alpha {from=""}
   ]
  }
  MaterialExpressionLinearInterpolate_2 {
   class=MaterialExpressionLinearInterpolate
   inputs [
    A {from=MaterialExpressionMultiply_0}
    B {from=MaterialExpressionNamedRerouteUsage_6}
    Alpha {from=MaterialExpressionNamedRerouteUsage_12}
   ]
  }
  MaterialExpressionMaterialFunctionCall_0 {
   class=MaterialExpressionMaterialFunctionCall
   inputs [
    Normal WS {from=MaterialExpressionNamedRerouteUsage_37}
    Specular {from=MaterialExpressionNamedRerouteUsage_10}
    Roughness {from=MaterialExpressionNamedRerouteUsage_9}
    Reflectance Mult {from=MaterialExpressionNamedRerouteUsage_71}
    Dielectric Opacity {from=MaterialExpressionNamedRerouteUsage_53}
    Diffuse Color {from=MaterialExpressionNamedRerouteUsage_17}
    Surface Opacity Mask {from=MaterialExpressionNamedRerouteUsage_55}
    Volume Transmittance {from=""}
   ]
   properties {materialFunction=/BRCosmetics/Accessories/FORT_Backpacks/Backpack_M_MED_DandyFlare/Materials/MaterialFunctions/MF_RecoverTranslucentReflection}
  }
  MaterialExpressionMaterialFunctionCall_13 {
   class=MaterialExpressionMaterialFunctionCall
   inputs [
    Viewport UV {from=MaterialExpressionReroute_0}
    Max Distance {from=""}
   ]
   properties {materialFunction=/Game/Global/MaterialFunctions/TechArt/Refraction/MF_DistanceToRefractedScene}
  }
  MaterialExpressionMaterialFunctionCall_14 {
   class=MaterialExpressionMaterialFunctionCall
   inputs [
    Viewport UV {from=""}
    Max Distance {from=""}
   ]
   properties {materialFunction=&MF_DistanceToRefractedScene}
  }
  MaterialExpressionMaterialFunctionCall_15 {
   class=MaterialExpressionMaterialFunctionCall
   inputs [
    Normal WS {from=MaterialExpressionNamedRerouteDeclaration_15}
    Sky Illuminance Mult {from=""}
    Sky Desaturation {from=""}
    Sun Illuminance Mult {from=""}
    Sun Desaturation {from=""}
   ]
   properties {materialFunction=/Game/Global/MaterialFunctions/TechArt/Refraction/MF_SunSkyProperties}
  }
  MaterialExpressionMaterialFunctionCall_18 {
   class=MaterialExpressionMaterialFunctionCall
   inputs [Opacity {from=MaterialExpressionGetMaterialAttributes_1, output=Opacity}]
   properties {materialFunction=/Game/Global/Materials/MaterialGym/Materials/SortedOutline/MF_SortTranslucency}
  }
  MaterialExpressionMaterialFunctionCall_20 {
   class=MaterialExpressionMaterialFunctionCall
   inputs [
    Sky Illuminance {from=MaterialExpressionNamedRerouteUsage_114}
    Sun Illuminance {from=MaterialExpressionNamedRerouteUsage_81}
    Sun Direction {from=MaterialExpressionNamedRerouteUsage_82}
    Scene Normal {from=MaterialExpressionNormalize_0}
    Scene Albedo {from=MaterialExpressionNamedRerouteUsage_95}
    Scene Color Scale {from=""}
   ]
   properties {materialFunction=/Game/Global/MaterialFunctions/TechArt/Refraction/MF_SceneLuminanceFromAlbedo}
  }
  MaterialExpressionMaterialFunctionCall_21 {
   class=MaterialExpressionMaterialFunctionCall
   inputs [
    Normal TS {from=MaterialExpressionNamedRerouteUsage_83}
    Refraction IOR {from=MaterialExpressionNamedRerouteUsage_84}
    Distortion Scale {from=MaterialExpressionNamedRerouteUsage_27}
    Depth Contribution {from=MaterialExpressionNamedRerouteUsage_0}
    Pixel Normal Offset Threshold {from=MaterialExpressionNamedRerouteUsage_1}
    Enable Refraction Mask {from=""}
   ]
   properties {materialFunction=/Game/Global/MaterialFunctions/TechArt/Refraction/MF_RefractionDistortion_ScreenSpace}
  }
  MaterialExpressionMaterialFunctionCall_4 {
   class=MaterialExpressionMaterialFunctionCall
   inputs [
    Normal {from=MaterialExpressionNamedRerouteUsage_20}
    Intensity {from=MaterialExpressionNamedRerouteUsage_69}
   ]
   properties {materialFunction=/Game/Global/MaterialFunctions/MF_NormalIntensity}
  }
  MaterialExpressionMaterialFunctionCall_8 {
   class=MaterialExpressionMaterialFunctionCall
   inputs [
    Specular Luminance {from=""}
    Diffuse Luminance {from=""}
    Surface Normal WS {from=MaterialExpressionNamedRerouteUsage_36}
    Surface Specular {from=MaterialExpressionNamedRerouteUsage_307}
    Surface Roughness {from=MaterialExpressionNamedRerouteUsage_30}
    Surface Opacity {from=MaterialExpressionNamedRerouteUsage_39}
    Refraction Vector WS {from=MaterialExpressionNamedRerouteUsage_35}
    Scene Luminance {from=MaterialExpressionNamedRerouteUsage_314}
    Interior Emission {from=MaterialExpressionConstant_14}
    Volume Thickness {from=MaterialExpressionNamedRerouteUsage_74}
    Volume Density {from=MaterialExpressionNamedRerouteUsage_61}
    Absorption Coeffs {from=MaterialExpressionNamedRerouteUsage_32}
    Scattering Coeffs {from=MaterialExpressionNamedRerouteUsage_34}
    Scattering Phase {from=MaterialExpressionNamedRerouteUsage_33}
    Is Scene Inside Volume {from=MaterialExpressionNamedRerouteUsage_76}
    Interface Thickness {from=MaterialExpressionNamedRerouteUsage_31}
    //---- ALPHA-BLENDING ONLY ----// {from=""}
    Haze Amount {from=MaterialExpressionNamedRerouteUsage_46}
    //----------- LIGHTING -----------// {from=""}
    Sky Illuminance {from=MaterialExpressionNamedRerouteUsage_54}
    Sun Illuminance {from=MaterialExpressionNamedRerouteUsage_48}
    Sun Direction {from=MaterialExpressionNamedRerouteUsage_49}
   ]
   properties {materialFunction=/BRCosmetics/Accessories/FORT_Backpacks/Backpack_M_MED_DandyFlare/Materials/MaterialFunctions/MF_DielectricInteriorVolume}
  }
  MaterialExpressionMaterialFunctionCall_9 {
   class=MaterialExpressionMaterialFunctionCall
   inputs [
    Normal {from=MaterialExpressionNamedRerouteUsage_176}
    Direction {from=MaterialExpressionNamedRerouteUsage_62}
    IOR {from=MaterialExpressionNamedRerouteUsage_177}
   ]
   properties {materialFunction=/Game/Global/MaterialFunctions/TechArt/Refraction/MF_ThicknessDirectional}
  }
  MaterialExpressionMax_0 {
   class=MaterialExpressionMax
   inputs [
    A {from=MaterialExpressionMultiply_15}
    B {from=""}
   ]
   properties {constB=0}
  }
  MaterialExpressionMax_2 {
   class=MaterialExpressionMax
   inputs [
    A {from=MaterialExpressionMultiply_5}
    B {from=""}
   ]
   properties {constB=0}
  }
  MaterialExpressionMax_9 {
   class=MaterialExpressionMax
   inputs [
    A {from=MaterialExpressionScalarParameter_40}
    B {from=""}
   ]
   properties {constB=0}
  }
  MaterialExpressionMultiply_0 {
   class=MaterialExpressionMultiply
   inputs [
    A {from=MaterialExpressionVectorParameter_0, output=RGB}
    B {from=MaterialExpressionReroute_4}
   ]
  }
  MaterialExpressionMultiply_1 {
   class=MaterialExpressionMultiply
   inputs [
    A {from=MaterialExpressionVectorParameter_1, output=RGB}
    B {from=MaterialExpressionVectorParameter_1, output=RGB}
   ]
  }
  MaterialExpressionMultiply_10 {
   class=MaterialExpressionMultiply
   inputs [
    A {from=MaterialExpressionNamedRerouteUsage_72}
    B {from=MaterialExpressionMultiply_9}
   ]
  }
  MaterialExpressionMultiply_11 {
   class=MaterialExpressionMultiply
   inputs [
    A {from=MaterialExpressionMultiply_13}
    B {from=MaterialExpressionShadingPathSwitch_0}
   ]
  }
  MaterialExpressionMultiply_13 {
   class=MaterialExpressionMultiply
   inputs [
    A {from=MaterialExpressionVectorParameter_2, output=RGB}
    B {from=MaterialExpressionVectorParameter_2, output=RGB}
   ]
  }
  MaterialExpressionMultiply_14 {
   class=MaterialExpressionMultiply
   inputs [
    A {from=MaterialExpressionChannelMaskParameter_0, output=R}
    B {from=MaterialExpressionScalarParameter_9}
   ]
  }
  MaterialExpressionMultiply_15 {
   class=MaterialExpressionMultiply
   inputs [
    A {from=MaterialExpressionVectorParameter_6, output=RGB}
    B {from=MaterialExpressionVectorParameter_6, output=RGB}
   ]
  }
  MaterialExpressionMultiply_19 {
   class=MaterialExpressionMultiply
   inputs [
    A {from=MaterialExpressionNamedRerouteUsage_127}
    B {from=MaterialExpressionSceneColor_4, output=RGB}
   ]
  }
  MaterialExpressionMultiply_2 {
   class=MaterialExpressionMultiply
   inputs [
    A {from=MaterialExpressionPower_1}
    B {from=MaterialExpressionNamedRerouteUsage_28}
   ]
  }
  MaterialExpressionMultiply_3 {
   class=MaterialExpressionMultiply
   inputs [
    A {from=MaterialExpressionNamedRerouteUsage_25}
    B {from=MaterialExpressionNamedRerouteUsage_24}
   ]
  }
  MaterialExpressionMultiply_31 {
   class=MaterialExpressionMultiply
   inputs [
    A {from=MaterialExpressionPower_5}
    B {from=MaterialExpressionNamedRerouteUsage_124}
   ]
  }
  MaterialExpressionMultiply_4 {
   class=MaterialExpressionMultiply
   inputs [
    A {from=MaterialExpressionEyeAdaptation_0, output=EyeAdaptation}
    B {from=MaterialExpressionConvert_4}
   ]
  }
  MaterialExpressionMultiply_5 {
   class=MaterialExpressionMultiply
   inputs [
    A {from=MaterialExpressionVectorParameter_7, output=RGB}
    B {from=MaterialExpressionVectorParameter_7, output=RGB}
   ]
  }
  MaterialExpressionMultiply_6 {
   class=MaterialExpressionMultiply
   inputs [
    A {from=MaterialExpressionNamedRerouteUsage_2}
    B {from=MaterialExpressionNamedRerouteUsage_3}
   ]
  }
  MaterialExpressionMultiply_7 {
   class=MaterialExpressionMultiply
   inputs [
    A {from=MaterialExpressionCameraVectorWS_5}
    B {from=""}
   ]
   properties {constB=-1}
  }
  MaterialExpressionMultiply_8 {
   class=MaterialExpressionMultiply
   inputs [
    A {from=MaterialExpressionScalarParameter_12}
    B {from=MaterialExpressionScalarParameter_0}
   ]
  }
  MaterialExpressionMultiply_9 {
   class=MaterialExpressionMultiply
   inputs [
    A {from=MaterialExpressionPower_4}
    B {from=MaterialExpressionNamedRerouteUsage_155}
   ]
  }
  MaterialExpressionNamedRerouteDeclaration_0 {
   class=MaterialExpressionNamedRerouteDeclaration
   inputs [None {from=MaterialExpressionScalarParameter_3}]
   properties {name="Distortion Depth Contribution (S)", nodeColor=<rgba 1 0.6094 0 1>}
  }
  MaterialExpressionNamedRerouteDeclaration_1 {
   class=MaterialExpressionNamedRerouteDeclaration
   inputs [None {from=MaterialExpressionFunctionInput_0}]
   properties {name="Material In (MA)", nodeColor=<rgba 1 0 0.2109 1>}
  }
  MaterialExpressionNamedRerouteDeclaration_10 {
   class=MaterialExpressionNamedRerouteDeclaration
   inputs [None {from=MaterialExpressionMaterialFunctionCall_8, output="Refraction Result - AlphaBlending (V3)"}]
   properties {name="Refraction Result (AlphaBlending) (V3)", nodeColor=<rgba 0 0.2435 1 1>}
  }
  MaterialExpressionNamedRerouteDeclaration_11 {
   class=MaterialExpressionNamedRerouteDeclaration
   inputs [None {from=MaterialExpressionMaterialFunctionCall_8, output="Material Opacity - AlphaBlending (S)"}]
   properties {name="Material Opacity (AlphaBlending) (S)", nodeColor=<rgba 0.9391 0 1 1>}
  }
  MaterialExpressionNamedRerouteDeclaration_12 {
   class=MaterialExpressionNamedRerouteDeclaration
   inputs [None {from=MaterialExpressionScalarParameter_10}]
   properties {name="Reflectance Mult Mobile (S)", nodeColor=<rgba 0.2891 0 1 1>}
  }
  MaterialExpressionNamedRerouteDeclaration_13 {
   class=MaterialExpressionNamedRerouteDeclaration
   inputs [None {from=MaterialExpressionScalarParameter_39}]
   properties {name="Thickness Exp (S)", nodeColor=<rgba 1 0.375 0 1>}
  }
  MaterialExpressionNamedRerouteDeclaration_14 {
   class=MaterialExpressionNamedRerouteDeclaration
   inputs [None {from=MaterialExpressionNormalize_1}]
   properties {name="Surface Normal WS (V3)", nodeColor=<rgba 0 0.5913 1 1>}
  }
  MaterialExpressionNamedRerouteDeclaration_15 {
   class=MaterialExpressionNamedRerouteDeclaration
   inputs [None {from=MaterialExpressionMultiply_7}]
   properties {name="Refraction Vector WS (V3)", nodeColor=<rgba 0 1 0.9766 1>}
  }
  MaterialExpressionNamedRerouteDeclaration_16 {
   class=MaterialExpressionNamedRerouteDeclaration
   inputs [None {from=MaterialExpressionGetMaterialAttributes_0, output=EmissiveColor}]
   properties {name="Emissive Color In (V3)", nodeColor=<rgba 0 0.2435 1 1>}
  }
  MaterialExpressionNamedRerouteDeclaration_161 {
   class=MaterialExpressionNamedRerouteDeclaration
   inputs [None {from=MaterialExpressionVectorParameter_10, output=Specular Exp}]
   properties {name="Specular Exp (S)", nodeColor=<rgba 1 0 0.3047 1>}
  }
  MaterialExpressionNamedRerouteDeclaration_17 {
   class=MaterialExpressionNamedRerouteDeclaration
   inputs [None {from=MaterialExpressionMultiply_6}]
   properties {name="Surface Emission (V3)", nodeColor=<rgba 0 0.2435 1 1>}
  }
  MaterialExpressionNamedRerouteDeclaration_19 {
   class=MaterialExpressionNamedRerouteDeclaration
   inputs [None {from=MaterialExpressionFunctionInput_1}]
   properties {name="Surface Opacity (S)", nodeColor=<rgba 0 1 0.087 1>}
  }
  MaterialExpressionNamedRerouteDeclaration_195 {
   class=MaterialExpressionNamedRerouteDeclaration
   inputs [None {from=MaterialExpressionStaticSwitch_8}]
   properties {name="Volume Density (S)", nodeColor=<rgba 0 0.3438 1 1>}
  }
  MaterialExpressionNamedRerouteDeclaration_2 {
   class=MaterialExpressionNamedRerouteDeclaration
   inputs [None {from=MaterialExpressionLinearInterpolate_2}]
   properties {name="Absorption Coeffs (V3)", nodeColor=<rgba 0.125 0 1 1>}
  }
  MaterialExpressionNamedRerouteDeclaration_20 {
   class=MaterialExpressionNamedRerouteDeclaration
   inputs [None {from=MaterialExpressionSaturate_1}]
   properties {name="Surface Specular (S)", nodeColor=<rgba 0.1953 0 1 1>}
  }
  MaterialExpressionNamedRerouteDeclaration_21 {
   class=MaterialExpressionNamedRerouteDeclaration
   inputs [None {from=MaterialExpressionScalarParameter_11}]
   properties {name="Roughness Mult Mobile (S)", nodeColor=<rgba 0.0312 1 0 1>}
  }
  MaterialExpressionNamedRerouteDeclaration_22 {
   class=MaterialExpressionNamedRerouteDeclaration
   inputs [None {from=MaterialExpressionScalarParameter_14}]
   properties {name="Volume - Haze Amount (S)", nodeColor=<rgba 0 1 0.3438 1>}
  }
  MaterialExpressionNamedRerouteDeclaration_23 {
   class=MaterialExpressionNamedRerouteDeclaration
   inputs [None {from=MaterialExpressionMultiply_11}]
   properties {name="Tinted Glass Absorption (V3)", nodeColor=<rgba 0 0.6953 1 1>}
  }
  MaterialExpressionNamedRerouteDeclaration_24 {
   class=MaterialExpressionNamedRerouteDeclaration
   inputs [None {from=MaterialExpressionGetMaterialAttributes_0, output=Specular}]
   properties {name="Specular In (S)", nodeColor=<rgba 1 0 0.3047 1>}
  }
  MaterialExpressionNamedRerouteDeclaration_25 {
   class=MaterialExpressionNamedRerouteDeclaration
   inputs [None {from=MaterialExpressionGetMaterialAttributes_0, output=Roughness}]
   properties {name="Roughness In (S)", nodeColor=<rgba 0 0.625 1 1>}
  }
  MaterialExpressionNamedRerouteDeclaration_26 {
   class=MaterialExpressionNamedRerouteDeclaration
   inputs [None {from=MaterialExpressionGetMaterialAttributes_0, output=Normal}]
   properties {name="Normal In TS (V3)", nodeColor=<rgba 0.9453 1 0 1>}
  }
  MaterialExpressionNamedRerouteDeclaration_27 {
   class=MaterialExpressionNamedRerouteDeclaration
   inputs [None {from=MaterialExpressionMultiply_14}]
   properties {name="Tinted Glass Mask (S)", nodeColor=<rgba 0 0.0391 1 1>}
  }
  MaterialExpressionNamedRerouteDeclaration_28 {
   class=MaterialExpressionNamedRerouteDeclaration
   inputs [None {from=MaterialExpressionMultiply_3}]
   properties {name="Surface BaseColor (V3)", nodeColor=<rgba 1 0.2109 0 1>}
  }
  MaterialExpressionNamedRerouteDeclaration_3 {
   class=MaterialExpressionNamedRerouteDeclaration
   inputs [None {from=MaterialExpressionMultiply_1}]
   properties {name="Scattering Coeffs (V3)", nodeColor=<rgba 0.1953 1 0 1>}
  }
  MaterialExpressionNamedRerouteDeclaration_31 {
   class=MaterialExpressionNamedRerouteDeclaration
   inputs [None {from=MaterialExpressionVectorParameter_10, output=Specular Mult}]
   properties {name="Specular Mult (S)", nodeColor=<rgba 1 0.2812 0 1>}
  }
  MaterialExpressionNamedRerouteDeclaration_33 {
   class=MaterialExpressionNamedRerouteDeclaration
   inputs [None {from=MaterialExpressionScalarParameter_30}]
   properties {name="Refraction Distortion Scale (S)", nodeColor=<rgba 0 0.2734 1 1>}
  }
  MaterialExpressionNamedRerouteDeclaration_4 {
   class=MaterialExpressionNamedRerouteDeclaration
   inputs [None {from=MaterialExpressionScalarParameter_1}]
   properties {name="Scattering Phase (S)", nodeColor=<rgba 0 1 0.6719 1>}
  }
  MaterialExpressionNamedRerouteDeclaration_41 {
   class=MaterialExpressionNamedRerouteDeclaration
   inputs [None {from=MaterialExpressionMax_9}]
   properties {name="Max Thickness (S)", nodeColor=<rgba 0.0547 0 1 1>}
  }
  MaterialExpressionNamedRerouteDeclaration_47 {
   class=MaterialExpressionNamedRerouteDeclaration
   inputs [None {from=MaterialExpressionScalarParameter_8}]
   properties {name="Asset Normal Intensity (S)", nodeColor=<rgba 0 1 0.6953 1>}
  }
  MaterialExpressionNamedRerouteDeclaration_5 {
   class=MaterialExpressionNamedRerouteDeclaration
   inputs [None {from=MaterialExpressionScalarParameter_2}]
   properties {name="Refraction IOR (S)", nodeColor=<rgba 0 0.1562 1 1>}
  }
  MaterialExpressionNamedRerouteDeclaration_51 {
   class=MaterialExpressionNamedRerouteDeclaration
   inputs [None {from=MaterialExpressionGetMaterialAttributes_0, output=BaseColor}]
   properties {name="BaseColor In (V3)", nodeColor=<rgba 1 0.2109 0 1>}
  }
  MaterialExpressionNamedRerouteDeclaration_54 {
   class=MaterialExpressionNamedRerouteDeclaration
   inputs [None {from=MaterialExpressionMaterialFunctionCall_15, output="Sun Direction (V3)"}]
   properties {name="Lighting - Sun Direction WS (V3)", nodeColor=<rgba 0 0.0625 1 1>}
  }
  MaterialExpressionNamedRerouteDeclaration_55 {
   class=MaterialExpressionNamedRerouteDeclaration
   inputs [None {from=MaterialExpressionMaterialFunctionCall_15, output="Sun Illuminance (V3)"}]
   properties {name="Lighting - Sun Illuminance (V3)", nodeColor=<rgba 0.7344 0 1 1>}
  }
  MaterialExpressionNamedRerouteDeclaration_56 {
   class=MaterialExpressionNamedRerouteDeclaration
   inputs [None {from=MaterialExpressionMaterialFunctionCall_15, output="Sky Illuminance (V3)"}]
   properties {name="Lighting - Sky Illuminance (V3)", nodeColor=<rgba 0 1 0.4141 1>}
  }
  MaterialExpressionNamedRerouteDeclaration_57 {
   class=MaterialExpressionNamedRerouteDeclaration
   inputs [None {from=MaterialExpressionMultiply_31}]
   properties {name="Asset Thickness VD (S)", desc="Thickness (View Dependent)", nodeColor=<rgba 1 0 0.5156 1>}
  }
  MaterialExpressionNamedRerouteDeclaration_58 {
   class=MaterialExpressionNamedRerouteDeclaration
   inputs [None {from=MaterialExpressionMax_0}]
   properties {name="Scene Tint (V3)", nodeColor=<rgba 0.7578 1 0 1>}
  }
  MaterialExpressionNamedRerouteDeclaration_6 {
   class=MaterialExpressionNamedRerouteDeclaration
   inputs [None {from=MaterialExpressionScalarParameter_4}]
   properties {name="Interface Thickness (S)", nodeColor=<rgba 1 0 0.7969 1>}
  }
  MaterialExpressionNamedRerouteDeclaration_61 {
   class=MaterialExpressionNamedRerouteDeclaration
   inputs [None {from=MaterialExpressionConvert_7}]
   properties {name="Is Scene Inside Volume (S)", nodeColor=<rgba 1 0 0.9844 1>}
  }
  MaterialExpressionNamedRerouteDeclaration_62 {
   class=MaterialExpressionNamedRerouteDeclaration
   inputs [None {from=MaterialExpressionConvert_7}]
   properties {name="Volume Thickness (S)", nodeColor=<rgba 0 0.5312 1 1>}
  }
  MaterialExpressionNamedRerouteDeclaration_67 {
   class=MaterialExpressionNamedRerouteDeclaration
   inputs [None {from=MaterialExpressionMax_2}]
   properties {name="Alpha Blending - Scene Avg Color (V3)", nodeColor=<rgba 0 0.625 1 1>}
  }
  MaterialExpressionNamedRerouteDeclaration_68 {
   class=MaterialExpressionNamedRerouteDeclaration
   inputs [None {from=MaterialExpressionShadingPathSwitch_3}]
   properties {name="Enable Alpha Blending (SB)", nodeColor=<rgba 0 0.9391 1 1>}
  }
  MaterialExpressionNamedRerouteDeclaration_7 {
   class=MaterialExpressionNamedRerouteDeclaration
   inputs [None {from=MaterialExpressionScalarParameter_5}]
   properties {name="Distortion Normal Threshold (S)", nodeColor=<rgba 0.2891 0 1 1>}
  }
  MaterialExpressionNamedRerouteDeclaration_8 {
   class=MaterialExpressionNamedRerouteDeclaration
   inputs [None {from=MaterialExpressionMaterialFunctionCall_8, output="Refraction Result - Opaque (V3)"}]
   properties {name="Refraction Result (SceneColor) (V3)", nodeColor=<rgba 0 0.2435 1 1>}
  }
  MaterialExpressionNamedRerouteDeclaration_9 {
   class=MaterialExpressionNamedRerouteDeclaration
   inputs [None {from=MaterialExpressionConstant_15}]
   properties {name="Material Opacity (SceneColor) (S)", nodeColor=<rgba 0.9391 0 1 1>}
  }
  MaterialExpressionNamedRerouteDeclaration_91 {
   class=MaterialExpressionNamedRerouteDeclaration
   inputs [None {from=MaterialExpressionNormalize_2}]
   properties {name="Surface Normal TS (V3)", nodeColor=<rgba 0 0.8594 1 1>}
  }
  MaterialExpressionNamedRerouteDeclaration_93 {
   class=MaterialExpressionNamedRerouteDeclaration
   inputs [None {from=MaterialExpressionSaturate_10}]
   properties {name="Surface Roughness (S)", nodeColor=<rgba 0 0.4844 1 1>}
  }
  MaterialExpressionNamedRerouteDeclaration_94 {
   class=MaterialExpressionNamedRerouteDeclaration
   inputs [None {from=MaterialExpressionConvert_4}]
   properties {name="Scene Distance (S)", nodeColor=<rgba 0 0.487 1 1>}
  }
  MaterialExpressionNamedRerouteDeclaration_95 {
   class=MaterialExpressionNamedRerouteDeclaration
   inputs [None {from=MaterialExpressionMultiply_4}]
   properties {name="Scene Luminance (V3)", nodeColor=<rgba 1 0.4453 0 1>}
  }
  MaterialExpressionNamedRerouteDeclaration_98 {
   class=MaterialExpressionNamedRerouteDeclaration
   inputs [None {from=MaterialExpressionVectorParameter_10, output=Roughness Exp}]
   properties {name="Roughness Exp (S)", nodeColor=<rgba 1 0.9844 0 1>}
  }
  MaterialExpressionNamedRerouteDeclaration_99 {
   class=MaterialExpressionNamedRerouteDeclaration
   inputs [None {from=MaterialExpressionVectorParameter_10, output=Roughness Mult}]
   properties {name="Roughness Mult (S)", nodeColor=<rgba 0.6641 0 1 1>}
  }
  MaterialExpressionNamedRerouteUsage_0 {class=MaterialExpressionNamedRerouteUsage}
  MaterialExpressionNamedRerouteUsage_1 {class=MaterialExpressionNamedRerouteUsage}
  MaterialExpressionNamedRerouteUsage_10 {class=MaterialExpressionNamedRerouteUsage}
  MaterialExpressionNamedRerouteUsage_113 {class=MaterialExpressionNamedRerouteUsage}
  MaterialExpressionNamedRerouteUsage_114 {class=MaterialExpressionNamedRerouteUsage}
  MaterialExpressionNamedRerouteUsage_12 {class=MaterialExpressionNamedRerouteUsage}
  MaterialExpressionNamedRerouteUsage_124 {class=MaterialExpressionNamedRerouteUsage}
  MaterialExpressionNamedRerouteUsage_127 {class=MaterialExpressionNamedRerouteUsage}
  MaterialExpressionNamedRerouteUsage_14 {class=MaterialExpressionNamedRerouteUsage}
  MaterialExpressionNamedRerouteUsage_154 {class=MaterialExpressionNamedRerouteUsage}
  MaterialExpressionNamedRerouteUsage_155 {class=MaterialExpressionNamedRerouteUsage}
  MaterialExpressionNamedRerouteUsage_158 {class=MaterialExpressionNamedRerouteUsage}
  MaterialExpressionNamedRerouteUsage_159 {class=MaterialExpressionNamedRerouteUsage}
  MaterialExpressionNamedRerouteUsage_17 {class=MaterialExpressionNamedRerouteUsage}
  MaterialExpressionNamedRerouteUsage_176 {class=MaterialExpressionNamedRerouteUsage}
  MaterialExpressionNamedRerouteUsage_177 {class=MaterialExpressionNamedRerouteUsage}
  MaterialExpressionNamedRerouteUsage_2 {class=MaterialExpressionNamedRerouteUsage}
  MaterialExpressionNamedRerouteUsage_20 {class=MaterialExpressionNamedRerouteUsage}
  MaterialExpressionNamedRerouteUsage_21 {class=MaterialExpressionNamedRerouteUsage}
  MaterialExpressionNamedRerouteUsage_22 {class=MaterialExpressionNamedRerouteUsage}
  MaterialExpressionNamedRerouteUsage_237 {class=MaterialExpressionNamedRerouteUsage}
  MaterialExpressionNamedRerouteUsage_24 {class=MaterialExpressionNamedRerouteUsage}
  MaterialExpressionNamedRerouteUsage_25 {class=MaterialExpressionNamedRerouteUsage}
  MaterialExpressionNamedRerouteUsage_27 {class=MaterialExpressionNamedRerouteUsage}
  MaterialExpressionNamedRerouteUsage_28 {class=MaterialExpressionNamedRerouteUsage}
  MaterialExpressionNamedRerouteUsage_3 {class=MaterialExpressionNamedRerouteUsage}
  MaterialExpressionNamedRerouteUsage_30 {class=MaterialExpressionNamedRerouteUsage}
  MaterialExpressionNamedRerouteUsage_307 {class=MaterialExpressionNamedRerouteUsage}
  MaterialExpressionNamedRerouteUsage_31 {class=MaterialExpressionNamedRerouteUsage}
  MaterialExpressionNamedRerouteUsage_314 {class=MaterialExpressionNamedRerouteUsage}
  MaterialExpressionNamedRerouteUsage_32 {class=MaterialExpressionNamedRerouteUsage}
  MaterialExpressionNamedRerouteUsage_33 {class=MaterialExpressionNamedRerouteUsage}
  MaterialExpressionNamedRerouteUsage_339 {class=MaterialExpressionNamedRerouteUsage}
  MaterialExpressionNamedRerouteUsage_34 {class=MaterialExpressionNamedRerouteUsage}
  MaterialExpressionNamedRerouteUsage_35 {class=MaterialExpressionNamedRerouteUsage}
  MaterialExpressionNamedRerouteUsage_36 {class=MaterialExpressionNamedRerouteUsage}
  MaterialExpressionNamedRerouteUsage_37 {class=MaterialExpressionNamedRerouteUsage}
  MaterialExpressionNamedRerouteUsage_39 {class=MaterialExpressionNamedRerouteUsage}
  MaterialExpressionNamedRerouteUsage_4 {class=MaterialExpressionNamedRerouteUsage}
  MaterialExpressionNamedRerouteUsage_41 {class=MaterialExpressionNamedRerouteUsage}
  MaterialExpressionNamedRerouteUsage_45 {class=MaterialExpressionNamedRerouteUsage}
  MaterialExpressionNamedRerouteUsage_46 {class=MaterialExpressionNamedRerouteUsage}
  MaterialExpressionNamedRerouteUsage_48 {class=MaterialExpressionNamedRerouteUsage}
  MaterialExpressionNamedRerouteUsage_49 {class=MaterialExpressionNamedRerouteUsage}
  MaterialExpressionNamedRerouteUsage_5 {class=MaterialExpressionNamedRerouteUsage}
  MaterialExpressionNamedRerouteUsage_53 {class=MaterialExpressionNamedRerouteUsage}
  MaterialExpressionNamedRerouteUsage_54 {class=MaterialExpressionNamedRerouteUsage}
  MaterialExpressionNamedRerouteUsage_55 {class=MaterialExpressionNamedRerouteUsage}
  MaterialExpressionNamedRerouteUsage_56 {class=MaterialExpressionNamedRerouteUsage}
  MaterialExpressionNamedRerouteUsage_59 {class=MaterialExpressionNamedRerouteUsage}
  MaterialExpressionNamedRerouteUsage_6 {class=MaterialExpressionNamedRerouteUsage}
  MaterialExpressionNamedRerouteUsage_61 {class=MaterialExpressionNamedRerouteUsage}
  MaterialExpressionNamedRerouteUsage_62 {class=MaterialExpressionNamedRerouteUsage}
  MaterialExpressionNamedRerouteUsage_63 {class=MaterialExpressionNamedRerouteUsage}
  MaterialExpressionNamedRerouteUsage_64 {class=MaterialExpressionNamedRerouteUsage}
  MaterialExpressionNamedRerouteUsage_65 {class=MaterialExpressionNamedRerouteUsage}
  MaterialExpressionNamedRerouteUsage_66 {class=MaterialExpressionNamedRerouteUsage}
  MaterialExpressionNamedRerouteUsage_67 {class=MaterialExpressionNamedRerouteUsage}
  MaterialExpressionNamedRerouteUsage_68 {class=MaterialExpressionNamedRerouteUsage}
  MaterialExpressionNamedRerouteUsage_69 {class=MaterialExpressionNamedRerouteUsage}
  MaterialExpressionNamedRerouteUsage_70 {class=MaterialExpressionNamedRerouteUsage}
  MaterialExpressionNamedRerouteUsage_71 {class=MaterialExpressionNamedRerouteUsage}
  MaterialExpressionNamedRerouteUsage_72 {class=MaterialExpressionNamedRerouteUsage}
  MaterialExpressionNamedRerouteUsage_73 {class=MaterialExpressionNamedRerouteUsage}
  MaterialExpressionNamedRerouteUsage_74 {class=MaterialExpressionNamedRerouteUsage}
  MaterialExpressionNamedRerouteUsage_76 {class=MaterialExpressionNamedRerouteUsage}
  MaterialExpressionNamedRerouteUsage_78 {class=MaterialExpressionNamedRerouteUsage}
  MaterialExpressionNamedRerouteUsage_81 {class=MaterialExpressionNamedRerouteUsage}
  MaterialExpressionNamedRerouteUsage_82 {class=MaterialExpressionNamedRerouteUsage}
  MaterialExpressionNamedRerouteUsage_83 {class=MaterialExpressionNamedRerouteUsage}
  MaterialExpressionNamedRerouteUsage_84 {class=MaterialExpressionNamedRerouteUsage}
  MaterialExpressionNamedRerouteUsage_9 {class=MaterialExpressionNamedRerouteUsage}
  MaterialExpressionNamedRerouteUsage_95 {class=MaterialExpressionNamedRerouteUsage}
  MaterialExpressionNamedRerouteUsage_98 {class=MaterialExpressionNamedRerouteUsage}
  MaterialExpressionNormalize_0 {class=MaterialExpressionNormalize, inputs [VectorInput {from=MaterialExpressionLinearInterpolate_1, output=""}]}
  MaterialExpressionNormalize_1 {class=MaterialExpressionNormalize, inputs [VectorInput {from=MaterialExpressionTransform_0, output=""}]}
  MaterialExpressionNormalize_2 {class=MaterialExpressionNormalize, inputs [VectorInput {from=MaterialExpressionMaterialFunctionCall_4, output="Result (V3)"}]}
  MaterialExpressionPower_1 {
   class=MaterialExpressionPower
   inputs [
    Base {from=MaterialExpressionNamedRerouteUsage_22}
    Exp {from=MaterialExpressionNamedRerouteUsage_339}
   ]
  }
  MaterialExpressionPower_4 {
   class=MaterialExpressionPower
   inputs [
    Base {from=MaterialExpressionNamedRerouteUsage_21}
    Exp {from=MaterialExpressionNamedRerouteUsage_154}
   ]
  }
  MaterialExpressionPower_5 {
   class=MaterialExpressionPower
   inputs [
    Base {from=MaterialExpressionMaterialFunctionCall_9, output="Thickness Normalized (S)"}
    Exp {from=MaterialExpressionNamedRerouteUsage_237}
   ]
   properties {desc="Normalized Thickness\r\n(View Dependent)"}
  }
  MaterialExpressionReroute_0 {class=MaterialExpressionReroute, inputs [None {from=MaterialExpressionMaterialFunctionCall_21, output="Refracted Viewport UV (V2)"}]}
  MaterialExpressionReroute_1 {class=MaterialExpressionReroute, inputs [None {from=MaterialExpressionStaticSwitch_2}]}
  MaterialExpressionReroute_2 {class=MaterialExpressionReroute, inputs [None {from=MaterialExpressionReroute_3}]}
  MaterialExpressionReroute_3 {class=MaterialExpressionReroute, inputs [None {from=MaterialExpressionConvert_8}]}
  MaterialExpressionReroute_4 {class=MaterialExpressionReroute, inputs [None {from=MaterialExpressionVectorParameter_0, output=Global Amount}]}
  MaterialExpressionReroute_5 {class=MaterialExpressionReroute, inputs [None {from=MaterialExpressionNamedRerouteUsage_56}]}
  MaterialExpressionSaturate_1 {class=MaterialExpressionSaturate, inputs [None {from=MaterialExpressionMultiply_2}]}
  MaterialExpressionSaturate_10 {class=MaterialExpressionSaturate, inputs [None {from=MaterialExpressionStaticSwitch_7}]}
  MaterialExpressionScalarParameter_0 {
   class=MaterialExpressionScalarParameter
   properties {defaultValue=1, group=Glass | 03_Volume, parameterName=Volume - Density, sortPriority=30}
  }
  MaterialExpressionScalarParameter_1 {
   class=MaterialExpressionScalarParameter
   properties {group=Glass | 03_Volume, parameterName=Volume - Scattering Phase, sliderMax=1, sliderMin=-1, sortPriority=70}
  }
  MaterialExpressionScalarParameter_10 {
   class=MaterialExpressionScalarParameter
   properties {defaultValue=1, group=Glass | 01_Surface, parameterName="Surface - Reflectance Multiplier [MOBILE]", sliderMax=10, sortPriority=30}
  }
  MaterialExpressionScalarParameter_11 {
   class=MaterialExpressionScalarParameter
   properties {defaultValue=1, group=Glass | 01_Surface, parameterName="Surface - Roughness Multiplier [MOBILE]", sliderMax=10, sortPriority=40}
  }
  MaterialExpressionScalarParameter_12 {
   class=MaterialExpressionScalarParameter
   properties {defaultValue=1, group=Glass | 03_Volume, parameterName="Volume - Density Gain [MOBILE]", sortPriority=40}
  }
  MaterialExpressionScalarParameter_14 {
   class=MaterialExpressionScalarParameter
   properties {group=Glass | 03_Volume, parameterName="Volume - Haze Amount [MOBILE]", sortPriority=80}
  }
  MaterialExpressionScalarParameter_2 {
   class=MaterialExpressionScalarParameter
   properties {defaultValue=1, group=Glass | 02_Refraction, parameterName=Refraction - IOR, sliderMax=5, sliderMin=1, sortPriority=10}
  }
  MaterialExpressionScalarParameter_3 {
   class=MaterialExpressionScalarParameter
   properties {
    defaultValue=0.1
    desc="How much depth should affect distortion scale. \r\n0: distortion amount is constant regardless of depth;\r\n1: far objects are more distorted than near ones (more accurate but can lead to extreme distortion)\r\n"
    group=Glass | 02_Refraction
    parameterName=Refraction - Distortion Depth Contribution
    sliderMax=1
    sortPriority=30
   }
  }
  MaterialExpressionScalarParameter_30 {
   class=MaterialExpressionScalarParameter
   properties {defaultValue=1, group=Glass | 02_Refraction, parameterName=Refraction - Distortion Scale, sliderMax=1, sortPriority=20}
  }
  MaterialExpressionScalarParameter_39 {
   class=MaterialExpressionScalarParameter
   properties {defaultValue=1, group=Glass | 03_Volume, parameterName=Volume - Thickness Gradient Exp, sliderMax=10, sliderMin=0.01, sortPriority=20}
  }
  MaterialExpressionScalarParameter_4 {
   class=MaterialExpressionScalarParameter
   properties {defaultValue=1, group=Glass | 02_Refraction, parameterName=Refraction - Interface Thickness, sortPriority=50}
  }
  MaterialExpressionScalarParameter_40 {
   class=MaterialExpressionScalarParameter
   properties {defaultValue=10, group=Glass | 03_Volume, parameterName=Volume - Max Thickness, sortPriority=10}
  }
  MaterialExpressionScalarParameter_5 {
   class=MaterialExpressionScalarParameter
   properties {
    defaultValue=0.05
    desc="Minimum difference between the vertex and pixel normals to produce distortion.\r\nCan be useful if the normal map is not very clean where it's supposed to be smooth (eg. 8bit artefacts)\r\n"
    group=Glass | 02_Refraction
    parameterName=Refraction - Distortion Normal Threshold
    sliderMax=2
    sortPriority=40
   }
  }
  MaterialExpressionScalarParameter_6 {
   class=MaterialExpressionScalarParameter
   properties {defaultValue=0.2, group=Glass | 03_Volume, parameterName="Volume - Tinted Glass Absorption Mult [MOBILE]", sortPriority=130}
  }
  MaterialExpressionScalarParameter_8 {
   class=MaterialExpressionScalarParameter
   properties {defaultValue=1, group=Glass | 01_Surface, parameterName=Surface - Asset Normal Intensity, sortPriority=10}
  }
  MaterialExpressionScalarParameter_9 {
   class=MaterialExpressionScalarParameter
   properties {group=Glass | 03_Volume, parameterName=Volume - Tinted Glass Mask Mult, sliderMax=1, sortPriority=110}
  }
  MaterialExpressionSceneColor_4 {
   class=MaterialExpressionSceneColor
   inputs [UVs {from=MaterialExpressionMaterialFunctionCall_21, output="Refracted Viewport UV (V2)"}]
   properties {}
  }
  MaterialExpressionSetMaterialAttributes_2 {
   class=MaterialExpressionSetMaterialAttributes
   inputs [
    MaterialAttributes {from=MaterialExpressionGetMaterialAttributes_1, output=MaterialAttributes}
    Opacity {from=MaterialExpressionMaterialFunctionCall_18, output=Result}
   ]
   properties {attributeSetTypes [#6]}
  }
  MaterialExpressionSetMaterialAttributes_4 {
   class=MaterialExpressionSetMaterialAttributes
   inputs [
    MaterialAttributes {from=MaterialExpressionNamedRerouteUsage_5, output=""}
    BaseColor {from=MaterialExpressionReroute_1, output=""}
    Metallic {from=MaterialExpressionConvert_11, output=""}
    Specular {from=MaterialExpressionConvert_11, output=""}
    Opacity {from=MaterialExpressionConvert_11, output=""}
    Roughness {from=MaterialExpressionNamedRerouteUsage_45, output=""}
    OpacityMask {from=MaterialExpressionConstant_10, output=""}
    Normal {from=MaterialExpressionNamedRerouteUsage_14, output=""}
    EmissiveColor {from=MaterialExpressionAdd_0, output=""}
   ]
   properties {attributeSetTypes [
     #1
     #7
     #3
     #6
     #4
     #8
     #5
     #2
    ]}
  }
  MaterialExpressionShadingPathSwitch_0 {
   class=MaterialExpressionShadingPathSwitch
   inputs [
    Default {from=MaterialExpressionConstant_0, output=""}
    Deferred {from="", output=""}
    Forward {from="", output=""}
    Mobile {from=MaterialExpressionScalarParameter_6, output=""}
   ]
   properties {}
  }
  MaterialExpressionShadingPathSwitch_3 {
   class=MaterialExpressionShadingPathSwitch
   inputs [
    Default {from=MaterialExpressionStaticSwitch_9, output=""}
    Deferred {from="", output=""}
    Forward {from="", output=""}
    Mobile {from=MaterialExpressionStaticBool_4, output=""}
   ]
   properties {}
  }
  MaterialExpressionStaticBoolParameter_9 {
   class=MaterialExpressionStaticBoolParameter
   inputs []
   properties {group=Glass | 99_DEV, parameterName=Refraction - Force Alpha Blending, sortPriority=10}
  }
  MaterialExpressionStaticBool_4 {class=MaterialExpressionStaticBool, properties {value=true}}
  MaterialExpressionStaticBool_5 {class=MaterialExpressionStaticBool}
  MaterialExpressionStaticSwitch_1 {
   class=MaterialExpressionStaticSwitch
   inputs [
    True {from=MaterialExpressionReroute_2, output=""}
    False {from=MaterialExpressionConvert_9, output=""}
    Value {from=MaterialExpressionReroute_5, output=""}
   ]
   properties {}
  }
  MaterialExpressionStaticSwitch_2 {
   class=MaterialExpressionStaticSwitch
   inputs [
    True {from=MaterialExpressionMaterialFunctionCall_0, output="Base Color (V3)"}
    False {from=MaterialExpressionNamedRerouteUsage_67, output=""}
    Value {from=MaterialExpressionNamedRerouteUsage_68, output=""}
   ]
   properties {}
  }
  MaterialExpressionStaticSwitch_3 {
   class=MaterialExpressionStaticSwitch
   inputs [
    True {from=MaterialExpressionNamedRerouteUsage_66, output=""}
    False {from=MaterialExpressionNamedRerouteUsage_65, output=""}
    Value {from=MaterialExpressionNamedRerouteUsage_70, output=""}
   ]
   properties {}
  }
  MaterialExpressionStaticSwitch_5 {
   class=MaterialExpressionStaticSwitch
   inputs [
    True {from=MaterialExpressionAppendVector_8, output=""}
    False {from=MaterialExpressionAppendVector_7, output=""}
    Value {from=MaterialExpressionNamedRerouteUsage_98, output=""}
   ]
   properties {}
  }
  MaterialExpressionStaticSwitch_7 {
   class=MaterialExpressionStaticSwitch
   inputs [
    True {from=MaterialExpressionMultiply_10, output=""}
    False {from=MaterialExpressionMultiply_9, output=""}
    Value {from=MaterialExpressionNamedRerouteUsage_78, output=""}
   ]
   properties {}
  }
  MaterialExpressionStaticSwitch_8 {
   class=MaterialExpressionStaticSwitch
   inputs [
    True {from=MaterialExpressionMultiply_8, output=""}
    False {from=MaterialExpressionScalarParameter_0, output=""}
    Value {from=MaterialExpressionNamedRerouteUsage_113, output=""}
   ]
   properties {}
  }
  MaterialExpressionStaticSwitch_9 {
   class=MaterialExpressionStaticSwitch
   inputs [
    True {from=MaterialExpressionStaticBool_4, output=""}
    False {from=MaterialExpressionStaticBool_5, output=""}
    Value {from=MaterialExpressionStaticBoolParameter_9, output=""}
   ]
   properties {}
  }
  MaterialExpressionTransform_0 {class=MaterialExpressionTransform, inputs [None {from=MaterialExpressionNormalize_2}]}
  MaterialExpressionVectorParameter_0 {
   class=MaterialExpressionVectorParameter
   properties {
    group=Glass | 03_Volume
    parameterName=Volume - Absorption Coeffs
    sortPriority=50
    channelNames {a=Global Amount, b=Absorption B, g=Absorption G, r=Absorption R}
   }
  }
  MaterialExpressionVectorParameter_1 {
   class=MaterialExpressionVectorParameter
   properties {
    group=Glass | 03_Volume
    parameterName=Volume - Scattering Coeffs
    sortPriority=60
    channelNames {a=Global Amount, b=Scattering B, g=Scattering G, r=Scattering R}
   }
  }
  MaterialExpressionVectorParameter_10 {
   class=MaterialExpressionVectorParameter
   properties {
    group=Glass | 01_Surface
    parameterName=Surface - Specular Controls
    sortPriority=20
    channelNames {a=Roughness Mult, b=Roughness Exp, g=Specular Mult, r=Specular Exp}
    defaultValue=<rgba 1 1 1 1>
   }
  }
  MaterialExpressionVectorParameter_2 {
   class=MaterialExpressionVectorParameter
   properties {
    group=Glass | 03_Volume
    parameterName=Volume - Tinted Glass Absorption Coeffs
    sortPriority=120
    channelNames {a=Global Amount, b=Absorption B, g=Absorption G, r=Absorption R}
   }
  }
  MaterialExpressionVectorParameter_6 {
   class=MaterialExpressionVectorParameter
   properties {
    group=Glass | 02_Refraction
    parameterName="Refraction - Refracted Scene Scale [PC]"
    sortPriority=50
    channelNames {a=Intensity, b=Tint B, g=Tint G, r=Tint R}
    defaultValue=<rgba 1 1 1 1>
   }
  }
  MaterialExpressionVectorParameter_7 {
   class=MaterialExpressionVectorParameter
   properties {
    group=Glass | 02_Refraction
    parameterName="Refraction - Refracted Scene Avg Color [MOBILE]"
    sortPriority=60
    channelNames {a=Intensity, b=Tint B, g=Tint G, r=Tint R}
    defaultValue=<rgba 0.18 0.18 0.18 1>
   }
  }
 }
 properties {previewBlendMode=BLEND_Translucent}
}
