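# Translucent parent material for the glass on the PixieParts backpack cosmetic, built on the 2023 character parent PreFX and PostFX functions.
# Surface parameters tint or replace albedo and derive opacity from a channel-selected mask with a minimum floor.
# An optional static-switched hologram function feeds into diffuse, opacity and emission.
# The result passes through the PixieParts glass function, with part of the emission split off and added back afterward.
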Material /BRCosmetics/Accessories/FORT_Backpacks/Backpack_M_MED_PixieParts_Lava/Materials/M_Backpack_PixieParts_Glass_Parent
graph {
 properties {
  allowTranslucentCustomDepthWrites=true
  bAutomaticallySetUsageInEditor=false
  bScreenSpaceReflections=true
  bUseMaterialAttributes=true
  bUsedWithMeshDeformer=true
  bUsedWithMorphTargets=true
  bUsedWithSkeletalMesh=true
  blendMode=BLEND_Translucent
  shadingModel=MSM_FromMaterialExpression
  translucencyLightingMode=TLM_SurfacePerPixelLighting
  usedShadingModels=Default Lit | Subsurface | Clear Coat | Subsurface Profile | Cloth
 }
 nodes [
  n1 = MaterialFunctionCall() {materialFunction=&MF_CharacterParent_2023_PreFX}
  MaterialFunctionCall(Material=NamedRerouteUsage(), M=@n1.M, Pre Skinned Local Position=@n1.M, Distance Blend=@n1.M, Fresnel=@n1.M) {materialFunction=&MF_CharacterParent_2023_PostFX}
  n2 = VectorParameter() {
   group=Glass | 01_Surface
   parameterName=Surface - Diffuse Albedo
   sortPriority=60
   channelNames {a="Tint <-> Replace", b=Constant B, g=Constant G, r=Constant R}
   defaultValue=<rgba 1 1 1 0>
  }
  NamedRerouteDeclaration(None=@n2.RGB) {name="Surface - Diffuse Albedo (V3)", nodeColor=<rgba 0.5938 0 1 1>}
  n3 = NamedRerouteUsage()
  n4 = NamedRerouteUsage()
  NamedRerouteDeclaration(None=StaticSwitchParameter(True=Max(A=@n3, B=@n4), False=@n4) {group=Glass | 00_Hologram | Global, parameterName=Enable Hologram, sortPriority=0}) {name="Surface - Opacity (S)", nodeColor=<rgba 0 1 0.2734 1>}
  n5 = ScalarParameter() {defaultValue=1, group=Glass | 01_Surface, parameterName=Surface - Opacity Mask Emission, sliderMax=1, sortPriority=130}
  NamedRerouteDeclaration(None=Saturate(None=@n5)) {name="Surface - Mask Emission by Opacity (S)", nodeColor=<rgba 1 0.9609 0 1>}
  n6 = MaterialFunctionCall!(Normal TS=NamedRerouteUsage()) {materialFunction=/BRCosmetics/Characters/Player/Male/Medium/Bodies/M_MED_PixieParts_Lava/Materials/MaterialFunctions/MF_PixieParts_GlassHologram}
  NamedRerouteDeclaration(None=@n6."Opacity (V3)") {name="Hologram Opacity (S)", nodeColor=<rgba 0 0.3131 1 1>}
  n7 = ScalarParameter() {
   desc="\r\nOpacity value where opacity mask is 0\r\n"
   group=Glass | 01_Surface
   parameterName=Surface - Min Opacity
   sliderMax=1
   sortPriority=100
  }
  n8 = ChannelMaskParameter(None=NamedRerouteUsage()) {group=Glass | 01_Surface, parameterName=Surface - Opacity Mask Channel, sortPriority=110}
  n9 = ScalarParameter() {defaultValue=1, group=Glass | 01_Surface, parameterName=Surface - Opacity Mask Mult, sliderMax=1, sortPriority=120}
  NamedRerouteDeclaration(None=Max(A=Saturate(None=@n7), B=Multiply(A=@n8.R, B=@n9))) {name="Surface - Opacity In (S)", nodeColor=<rgba 0 1 0.2734 1>}
  n10 = NamedRerouteUsage()
  n11 = NamedRerouteUsage()
  n12 = Saturate(None=LinearInterpolate(A=Multiply(A=@n10, B=@n11), B=@n11, Alpha=NamedRerouteUsage()))
  n13 = NamedRerouteUsage()
  n14 = NamedRerouteUsage()
  NamedRerouteDeclaration(None=StaticSwitchParameter(True=LinearInterpolate(A=@n12, B=@n13, Alpha=@n14), False=@n12) {group=Glass | 00_Hologram | Global, parameterName=Enable Hologram, sortPriority=0}) {name="Surface - Diffuse (V3)", nodeColor=<rgba 0.5938 0 1 1>}
  n15 = NamedRerouteUsage()
  n16 = NamedRerouteUsage()
  NamedRerouteDeclaration(None=StaticSwitchParameter(True=Add(A=@n15, B=@n16), False=@n16) {group=Glass | 00_Hologram | Global, parameterName=Enable Hologram, sortPriority=0}) {name="Surface - Emission (V3)", nodeColor=<rgba 1 0 0.75 1>}
  n17 = GetMaterialAttributes(None=NamedRerouteDeclaration(None=@n1.Result) {name="Material In (MA)", nodeColor=<rgba 0.4531 1 0 1>}) {attributeGetTypes [#1, #2, #3]}
  NamedRerouteDeclaration(None=@n17.Normal) {name="Normal TS In (V3)", nodeColor=<rgba 1 0 0.3984 1>}
  NamedRerouteDeclaration(None=@n6."Emission (V3)") {name="Hologram Emission (V3)", nodeColor=<rgba 0 0.3131 1 1>}
  NamedRerouteDeclaration(None=@n6."Diffuse (V3)") {name="Hologram Diffuse (V3)", nodeColor=<rgba 0 0.3131 1 1>}
  NamedRerouteDeclaration(None=@n1.Tech Art Mask) {name="FX Mask In (V3)", nodeColor=<rgba 0.5703 0 1 1>}
  n18 = NamedRerouteUsage()
  n19 = NamedRerouteUsage()
  n20 = Multiply(A=NamedRerouteUsage(), B=NamedRerouteUsage())
  n21 = GetMaterialAttributes(None=MaterialFunctionCall!(
    Material=SetMaterialAttributes!(MaterialAttributes=@n18, Base Color=@n19, Emissive Color=@n20) {attributeSetTypes [#1, #2]}
    TechArt Mask=NamedRerouteUsage()
    Surface Opacity=NamedRerouteUsage()
   ) {materialFunction=/BRCosmetics/Characters/Player/Male/Medium/Bodies/M_MED_PixieParts_Lava/Materials/MaterialFunctions/MF_PixieParts_Glass}."Result (MA)") {attributeGetTypes [#2]}
  n22 = Multiply(A=NamedRerouteUsage(), B=OneMinus(None=NamedRerouteUsage()))
  NamedRerouteDeclaration(None=SetMaterialAttributes!(MaterialAttributes=@n21.MaterialAttributes, Emissive Color=Add(A=@n21.Emissive Color, B=@n22)) {attributeSetTypes [#2]}) {name="Material Out (MA)", nodeColor=<rgba 0 1 0.1094 1>}
  NamedRerouteDeclaration(None=Saturate(None=@n2."Tint <-> Replace")) {name="Surface - Diffuse Albedo Mode (S)", nodeColor=<rgba 0.5938 0 1 1>}
  NamedRerouteDeclaration(None=@n17.Base Color) {name="BaseColor In (V3)", nodeColor=<rgba 1 0 0.3984 1>}
  NamedRerouteDeclaration(None=@n17.Emissive Color) {name="EmissiveColor In (V3)", nodeColor=<rgba 1 0 0.3984 1>}
 ]
}
