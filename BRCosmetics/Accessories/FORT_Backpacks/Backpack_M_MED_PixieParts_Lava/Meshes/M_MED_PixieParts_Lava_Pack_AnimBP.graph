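# Anim Blueprint for the PixieParts Lava backpack that adds secondary wing motion to the cosmetic.
# The input pose gets a small additive offset on the pack bone, then two chained Anim Dynamics nodes move the wings as angle-locked linear springs.
# Each update derives backpack rotation and translation variables from upper-spine yaw relative to the reference pose, but the AnimGraph shows no use of them.
# A body-type/gender check sits unconnected in the EventGraph.
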AnimBlueprint /BRCosmetics/Accessories/FORT_Backpacks/Backpack_M_MED_PixieParts_Lava/Meshes/M_MED_PixieParts_Lava_Pack_AnimBP
target_skeleton=&/M_MED_PixieParts_Lava_Pack_Skeleton
graphs [
 AnimGraph {nodes [
   n1 = "Transform (Modify) Bone - Bone: pack"!(ComponentPose=Local To Component(LocalPose=Input Pose()), bAlphaBoolEnabled:bool, Translation="0,2.3,0", Rotation="0,0,0", Scale) -> (Pose) {
    alphaInputType=Bool
    rotationMode=BMM_Additive
    translationMode=BMM_Additive
    translationSpace=BCS_BoneSpace
    alphaBoolBlend {blendOption=Linear}
    boneToModify {boneName=pack}
    translation {y=2.3}
   }
   n2 = Anim Dynamics - Bone: dyn__wing_l_01(ComponentPose=@n1, Alpha=1, ExternalForce="0,0,0") {
    bLinearSpring=true
    bOverrideLinearDamping=true
    bUseGravityOverride=true
    linearDampingOverride=0.999
    linearSpringConstant=180
    simulationSpace=BoneRelative
    boundBone {boneName=dyn__wing_l_01}
    componentAppliedLinearAccClamp=<111 111 111>
    componentLinearAccScale=<0.13 0.13>
    componentLinearVelScale=<0.13 0.13>
    gravityOverride {z=-10}
    physicsBodyDefinitions [{
      boundBone {boneName=dyn__wing_l_01}
      boxExtents=<60 60 60>
      constraintSetup {angularLimitsMax=<0 0 0>, angularLimitsMin=<0 0 0>, angularTarget=<0 0 0>, linearAxesMax=<3 0 3>, linearAxesMin=<-4 0 -3>}
      localJointOffset=<0.3 0.3 0>
     }]
    relativeSpaceBone {boneName=root}
    simSpaceSettings {maxAngularAcceleration=100, maxAngularVelocity=100, simSpaceAngularAlpha=0.17}
   }
   n3 = Anim Dynamics - Bone: dyn__wing_r_01(ComponentPose=@n2, Alpha=1, ExternalForce="0,0,0") {
    bLinearSpring=true
    bOverrideLinearDamping=true
    bUseGravityOverride=true
    linearDampingOverride=0.999
    linearSpringConstant=180
    simulationSpace=RootRelative
    boundBone {boneName=dyn__wing_r_01}
    componentAppliedLinearAccClamp=<111 111 111>
    componentLinearAccScale=<0.13 0.13>
    componentLinearVelScale=<0.13 0.13>
    gravityOverride {z=-10}
    physicsBodyDefinitions [{
      boundBone {boneName=dyn__wing_r_01}
      boxExtents=<69 69 69>
      constraintSetup {angularLimitsMax=<0 0 0>, angularLimitsMin=<0 0 0>, angularTarget=<0 0 0>, linearAxesMax=<3 0 3>, linearAxesMin=<-4 0 -3>}
      localJointOffset=<0.3 0.3 0>
     }]
    simSpaceSettings {maxAngularAcceleration=100, maxAngularVelocity=100, simSpaceAngularAlpha=0.17}
   }
   Output Pose(Result=Component To Local(ComponentPose=@n3)) {layerGroup=None}
  ]}
 EventGraph {nodes [
   Event BlueprintUpdateAnimation()
   n1 = GetOwnerSkeletalMeshComponent()
   n2 = BreakRotator(InRot=BreakTransform(InTransform=GetDeltaTransformFromRefPose(self=@n1, BoneName=spine_05, BaseName=spine_01)).Rotation)
   n3 = MapRangeClamped(Value=@n2.Yaw, InRangeA=-10, InRangeB=-50, OutRangeB=35)
   > Set backpack_rotate(backpack_rotate=MakeRotator(Roll=@n3, Pitch=float / float(A=@n3, B=-10)))
   > Set backpack_translation(backpack_translation=MakeVector(Y=MapRangeClamped(Value=@n2.Yaw, InRangeA=-15, InRangeB=-50, OutRangeB=-15), Z=-6.5))
   TryGetPawnOwner()
   Event OnCharacterCustomizationLoaded()
   n4 = "Get (a copy)"(Array:Array of Custom Character Part Object References, Dimension 1=1) -> (Output:Custom Character Part Object Reference)
   n5 = Get GenderPermitted!(self:Custom Character Part Object Reference=@n4) -> (GenderPermitted:EFortCustomGender Enum=Invalid)
   n6 = Get BodyTypesPermitted!(self:Custom Character Part Object Reference=@n4) -> (BodyTypesPermitted:EFortCustomBodyType Enum=NONE)
   Set isM_Med(isM_Med:bool=AND Boolean(
     A="Equal (Enum)"(A:EFortCustomBodyType Enum=@n6, B:EFortCustomBodyType Enum=Medium)
     B="Equal (Enum)"(A:EFortCustomGender Enum=@n5, B:EFortCustomGender Enum=Female)
    )) -> (Output_Get:bool)
  ]}
]
variables [
 backpack_rotate {type=Rotator}
 backpack_translation {type=Vector}
 Is Idle Female {type=boolean}
 isM_Med {type=boolean}
]
